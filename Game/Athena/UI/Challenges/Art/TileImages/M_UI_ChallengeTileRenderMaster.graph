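# Translucent UI master material for challenge tile images, showing a DisplayImage texture tinted by an accent color gradient.
# The UseHover static switch adds a time-driven sine bob to the image's vertical UVs.
# Offset, scaled and rotated UVs form the gradient, with AnimateGradient blending its start and end ranges to set tint strength before a slight contrast power.
# Image alpha is scaled by MasterAlpha.
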Material /Game/Athena/UI/Challenges/Art/TileImages/M_UI_ChallengeTileRenderMaster
graph {
 properties {
  bUsedWithStaticMesh=false
  blendMode=BLEND_Translucent
  materialDomain=MD_UI
  pixelDepthOffsetMode=PDOM_Legacy
  usedShadingModels=Default Lit
 }
 nodes [
  n1 = Time()
  n2 = ScalarParameter() {defaultValue=0.25, parameterName=TimeMult}
  n3 = Add(A=Sine(None=Multiply(A=@n1, B=@n2) {constB=0.2}))
  n4 = AppendVector(
   A=Constant()
   B=LinearInterpolate(
    A=ScalarParameter() {defaultValue=-0.05, parameterName=HoverRangeLow}
    B=ScalarParameter() {defaultValue=0.05, parameterName=HoverRangeHigh}
    Alpha=Multiply(A=@n3) {constB=0.5}
   )
  )
  n5 = TextureCoordinate()
  n6 = TextureSampleParameter2D(UVs=StaticSwitchParameter(True=Add(A=@n4, B=@n5), False=@n5) {parameterName=UseHover}) {parameterName=DisplayImage, texture=&/T_UI_ChallengeTile_Drift}
  LinearInterpolate(B=ScalarParameter() {defaultValue=1, parameterName=MasterAlpha}, Alpha=@n6.A)
  n7 = Add(
   A=TextureCoordinate()
   B=AppendVector(
    A=ScalarParameter() {group=ColorGradient, parameterName=GradientPosition_Y}
    B=ScalarParameter() {group=ColorGradient, parameterName=GradientPosition_X}
   )
  )
  n8 = ScalarParameter() {defaultValue=1, group=ColorGradient, parameterName=GradientScale}
  n9 = ScalarParameter() {defaultValue=8, group=ColorGradient, parameterName=Rotation_Gradient}
  n10 = ComponentMask(None=Rotator(Coordinate=Divide(A=@n7, B=@n8), Time=Divide(A=@n9) {constB=360}) {speed=6.2832}) {g=true}
  n11 = LinearInterpolate(
   A=ScalarParameter() {group=ColorGradient, parameterName=GradientStartMin}
   B=ScalarParameter() {defaultValue=0.25, group=ColorGradient, parameterName=GradientStartMax}
   Alpha=Clamp(None=@n10)
  )
  n12 = LinearInterpolate(
   A=ScalarParameter() {defaultValue=0.25, group=ColorGradient, parameterName=GradientEndMin, sliderMax=1, sliderMin=0.25}
   B=ScalarParameter() {defaultValue=0.25, group=ColorGradient, parameterName=GradientEndMax, sliderMax=1, sliderMin=0.25}
   Alpha=Clamp(None=@n10)
  )
  n13 = ScalarParameter() {parameterName=AnimateGradient}
  Power(Base=LinearInterpolate(
    A=@n6.RGB
    B=VectorParameter() {group=ColorGradient, parameterName=Accent, defaultValue=<rgba 1 0.007 0.1683 1>}.RGB
    Alpha=LinearInterpolate(A=@n11, B=@n12, Alpha=@n13)
   )) {constExponent=1.125}
 ]
}
