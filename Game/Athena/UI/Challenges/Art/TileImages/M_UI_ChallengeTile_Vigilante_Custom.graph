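# UI material for the Vigilante custom challenge tile image, set to translucent blending.
# A RadialGradientExponential centered horizontally and below middle drives two color lerps, one light gray to white, the other involving a dark red.
# The UV vertical component is scaled, saturated and remapped by a lerp into a partial-range gradient.
# The graph shows no texture samples, parameters or material output connections, and includes loose gray and red color constants.
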Material /Game/Athena/UI/Challenges/Art/TileImages/M_UI_ChallengeTile_Vigilante_Custom
graph {
 properties {
  bUsedWithStaticMesh=false
  blendMode=BLEND_Translucent
  materialDomain=MD_UI
  pixelDepthOffsetMode=PDOM_Legacy
  usedShadingModels=Default Lit
 }
 nodes [
  Constant3Vector() {constant=<rgba 0.0075 0.0075 0.0075 0>}
  Constant3Vector() {constant=<rgba 0.75 0.0189 0.0181 0>}
  Constant3Vector() {constant=<rgba 0.075 0.0019 0.0018 0>}
  n1 = MaterialFunctionCall(UVs=TextureCoordinate(), CenterPosition=Constant2Vector() {g=0.7, r=0.5}, Radius=Constant() {r=0.666}, Density=Constant() {r=1.666}) {materialFunction=&RadialGradientExponential}
  LinearInterpolate(A=Constant3Vector() {constant=<rgba 0.75 0.75 0.75 0>}, B=Constant3Vector() {constant=<rgba 1 1 1 0>}, Alpha=@n1.RadialGradientExponential)
  LinearInterpolate(A=Constant3Vector(), B=Constant3Vector() {constant=<rgba 0.0675 0.0048 0.0069 0>}, Alpha=@n1.RadialGradientExponential)
  n2 = Multiply(A=ComponentMask(None=TextureCoordinate()) {g=true}) {constB=1.375}
  Saturate(None=LinearInterpolate(A=Constant() {r=0.4}, B=Constant() {r=0.875}, Alpha=Saturate(None=@n2)))
 ]
}
